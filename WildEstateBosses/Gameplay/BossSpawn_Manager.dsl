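; Boss spawn manager for Wild Estate that assigns boss tags to placed ChallengeManager progress-bar actors and initializes them to spawn bosses.
; On authority, it spawns forced tags first, then random managers up to SpawnCount, or waits for safe zone state changes to pick managers inside the next zone.
; Random bosses draw from a shrinking enabled-tag pool, optionally filtered by location data rows.
; Debug events spawn one chosen boss at every remaining manager.
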(event Custom|SpawnBossAtAllLocations
  (params (Boss%20IDTag :type Gameplay%20Tag%20Structure))
  (for _array_element (Variables|Default|GetBossChallengeManagers)
    (Variables|/AIProgressBar/Blueprints/ChallengeManager.ChallengeManager_C|SetBossIDTag Boss%20IDTag _array_element)
    (Class|ChallengeManager|Init _array_element)))

(event Custom|TestBossAtAllLocations
  (params (BossName :type String))
  (switch string BossName
    (:Evil%20Homer
      (Variables|Default|SetForceRandomBossSpawnTag "(TagName=\"AISpawnerData.Type.Boss.WildEstate.SilentNovel.VaneHush\")"))
    (:Witch%20Marge
      (Variables|Default|SetForceRandomBossSpawnTag "(TagName=\"AISpawnerData.Type.Boss.WildEstate.FruitFire.NudgeHonk\")"))
    (:Krusty
      (Variables|Default|SetForceRandomBossSpawnTag "(TagName=\"AISpawnerData.Type.Boss.WildEstate.ClimbSpill\")"))
    (:Moe
      (Variables|Default|SetForceRandomBossSpawnTag "(TagName=\"AISpawnerData.Type.Boss.WildEstate.DuckCoast\")"))
    (:Scratchy
      (Variables|Default|SetForceRandomBossSpawnTag "(TagName=\"AISpawnerData.Type.Boss.WildEstate.TroopFactor\")")))
  (CallFunction|SpawnBossatAllLocations (Variables|Default|GetForceRandomBossSpawnTag)))

(event EventBeginPlay
  (|CallParentFunction)
  (Networking|SwitchHasAuthority
    (:Authority
      (if (ScalableFloat|GetValueasBool (Variables|SafeZone|GetUseSafeZoneUpdateToSpawn))
        ;[0 0 0 0.5] Initialize Safe Zone Indicator
        (Class|FortAsyncActionCurrentPlaylistReady|PlaylistDataReadyAsync
          (:OnReady
            (bind _returnvalue (GameplayEventRouter|GetEventRouter "/Script/FortniteGame.FortGameState" self))
            (bind (_asynctaskproxy _payload) (GameplayEventRouter|StartListeningToStatefulEvent _returnvalue "/Script/FortniteGame.FortSafeZoneIndicatorUpdatedEvent")
              (:then
                (Variables|SafeZone|SetSafeZoneIndicatorUpdatedAsyncTask _asynctaskproxy)
                (Utilities|FlowControl|DelayUntilNextTick)
                (CallFunction|InitializeBossManager))
              (:OnEventReceived
                _)
              (:OnSavedEventStateExists
                _))
            (CallFunction|OnSafeZoneIndicatorSpawned (Utilities|Struct|BreakFortSafeZoneIndicatorUpdatedEvent _payload))))
        (else
          (Utilities|FlowControl|DelayUntilNextTick)
          (CallFunction|InitializeBossManager))))))

(event Custom|SetPossibleBossSpawnTags
  (for _array_element (Variables|Default|GetValidBossTags)
    (if (ScalableFloat|GetValueasBool (Utilities|Struct|BreakValidBossStruct _array_element))
      (GameplayTags|AddGameplayTag (Variables|Default|GetPossibleBossSpawnTags) (.AssociatedBossTag_8_A43A80F049F2C88B343C20BBDFE7981E (Utilities|Struct|BreakValidBossStruct _array_element)) :!class /Script/GameplayTags.BlueprintGameplayTagLibrary))))

(event Custom|InitializeBossManager
  (CallFunction|SetPossibleBossSpawnTags)
  (bind _outactors (Actor|GetAllActorsOfClass "/AIProgressBar/Blueprints/ChallengeManager_SimpleBossProgressBar.ChallengeManager_SimpleBossProgressBar_C"))
  (Variables|Default|SetBossChallengeManagers _outactors)
  (CallFunction|InitializeChallengeManagers))

(event Custom|InitializeChallengeManagers
  ;[0 0 0 0.5] Conditionally spawn bosses when the safe zone is updated. Otherwise spawn all bosses on BeginPlay
  (if (or (not (ScalableFloat|GetValueasBool (Variables|SafeZone|GetUseSafeZoneUpdateToSpawn))) (Variables|Default|GetDebug))
    (CallFunction|RequestedSpawns)
    (CallFunction|SpawntoCount (ScalableFloat|GetValueasInteger (Variables|Default|GetSpawnCount)))))

(event Custom|RequestedSpawns
  (bind _bosschallengemanagers (Variables|Default|GetBossChallengeManagers))
  (if (Utilities|Array|IsNotEmpty (GameplayTags|BreakGameplayTagContainer (Variables|Default|GetForceSpawnBossTags)))
    (bind _array_element (Utilities|Array|ForEachLoopwithBreak _bosschallengemanagers)
      (:LoopBody
        (if (GameplayTags|TagContainer|CompareTagContainertoOtherTags (Variables|/AIProgressBar/Blueprints/ChallengeManager.ChallengeManager_C|GetBossIDTag _array_element) (Variables|Default|GetForceSpawnBossTags))
          (if (Utilities|IsRandomBossManager _array_element)
            (CallFunction|SetRandomBoss _array_element)
            (else
              _))
          (Class|ChallengeManager|Init _array_element)
          (Math|Integer|IncrementInt (Variables|Default|GetCountSpawned)
            (:%20%20
              (Utilities|Array|Add (Variables|Default|GetSpawnedChallengeManager) _array_element)
              (GameplayTags|RemoveGameplayTag (Variables|Default|GetForceSpawnBossTags) (Variables|/AIProgressBar/Blueprints/ChallengeManager.ChallengeManager_C|GetBossIDTag _array_element) :!class /Script/GameplayTags.BlueprintGameplayTagLibrary)
              (GameplayTags|AddGameplayTag (Variables|Default|GetSpawnedBossTags) (Variables|/AIProgressBar/Blueprints/ChallengeManager.ChallengeManager_C|GetBossIDTag _array_element) :!class /Script/GameplayTags.BlueprintGameplayTagLibrary)
              (GameplayTags|RemoveGameplayTag (Variables|Default|GetPossibleBossSpawnTags) (Variables|/AIProgressBar/Blueprints/ChallengeManager.ChallengeManager_C|GetBossIDTag _array_element) :!class /Script/GameplayTags.BlueprintGameplayTagLibrary)))))
      (:Completed
        ;[1 1 1] Remove all managers with tags of spawned bosses
        (bind (_arrayindex _arrayelement) (Utilities|Array|ReverseforEachLoop _bosschallengemanagers)
          (:LoopBody
            (Utilities|IsValid _arrayelement
              (:Is%20Valid
                (if (GameplayTags|TagContainer|CompareTagContainertoOtherTags (Variables|/AIProgressBar/Blueprints/ChallengeManager.ChallengeManager_C|GetBossIDTag _arrayelement) (Variables|Default|GetSpawnedBossTags))
                  (Utilities|Array|RemoveIndex (Variables|Default|GetBossChallengeManagers) _arrayindex))))))))))

(event Custom|InitializeChallengeManager
  (params (ChallengeManager :type Challenge%20Manager%20Simple%20Boss%20Progress%20Bar%20Object%20Reference))
  (bind _bossidtag (Variables|/AIProgressBar/Blueprints/ChallengeManager.ChallengeManager_C|GetBossIDTag ChallengeManager))
  (if (Utilities|IsRandomBossManager ChallengeManager)
    (CallFunction|SetRandomBoss ChallengeManager)
    (else
      _))
  (if (not (or (GameplayTags|TagContainer|CompareTagContainertoOtherTags (Variables|/AIProgressBar/Blueprints/ChallengeManager.ChallengeManager_C|GetBossIDTag ChallengeManager) (Variables|Default|GetSpawnedBossTags)) (== (Variables|/AIProgressBar/Blueprints/ChallengeManager.ChallengeManager_C|GetBossIDTag ChallengeManager) 0)))
    (GameplayTags|AddGameplayTag (Variables|Default|GetSpawnedBossTags) _bossidtag :!class /Script/GameplayTags.BlueprintGameplayTagLibrary)
    (GameplayTags|RemoveGameplayTag (Variables|Default|GetPossibleBossSpawnTags) _bossidtag :!class /Script/GameplayTags.BlueprintGameplayTagLibrary)
    (Class|ChallengeManager|Init ChallengeManager)
    (Math|Integer|IncrementInt (Variables|Default|GetCountSpawned)
      (:%20%20
        (Utilities|Array|Add (Variables|Default|GetSpawnedChallengeManager) ChallengeManager))))
  (Utilities|Array|RemoveItem (Variables|Default|GetBossChallengeManagers) ChallengeManager)
  (if (Variables|Default|GetDebug)
    (Development|PrintString (Utilities|String|Append "Spawned Boss: " (GameplayTags|GetDebugStringfromGameplayTag (Variables|/AIProgressBar/Blueprints/ChallengeManager.ChallengeManager_C|GetBossIDTag ChallengeManager))) true true "(R=1.000000,G=0.074966,B=0.000000,A=1.000000)" 6e+01)))

(event Custom|SpawnToCount
  (params (Count :type Integer))
  (while (and (< (Variables|Default|GetCountSpawned) Count) (Utilities|Array|IsNotEmpty (Variables|Default|GetBossChallengeManagers)))
    (CallFunction|InitializeRandomManager))
  (if (Variables|Default|GetDebug)
    (for _array_element (Variables|Default|GetSpawnedChallengeManager)
      (Development|PrintString (GameplayTags|GetDebugStringfromGameplayTag (Variables|/AIProgressBar/Blueprints/ChallengeManager.ChallengeManager_C|GetBossIDTag _array_element))))))

(event Custom|InitializeRandomManager
  (bind _bosschallengemanagers (Variables|Default|GetBossChallengeManagers))
  (Variables|Default|SetCheckCount (Variables|Default|GetCountSpawned))
  (if (Utilities|Array|IsNotEmpty _bosschallengemanagers)
    (CallFunction|InitializeChallengeManager (Utilities|Array|RandomArrayItem _bosschallengemanagers))
    (if (not (< (Variables|Default|GetCheckCount) (Variables|Default|GetCountSpawned)))
      (CallFunction|InitializeRandomManager))))

(event Custom|SetRandomBoss
  (params (ChallengeManager :type Challenge%20Manager%20Simple%20Boss%20Progress%20Bar%20Object%20Reference))
  (bind _validbosstag (CallFunction|GetValidBossTag ChallengeManager))
  (bind _output_get (Variables|/AIProgressBar/Blueprints/ChallengeManager.ChallengeManager_C|SetBossIDTag (select (GameplayTags|IsGameplayTagValid (Variables|Default|GetForceRandomBossSpawnTag)) (Variables|Default|GetForceRandomBossSpawnTag) _validbosstag) ChallengeManager))
  (GameplayTags|RemoveGameplayTag (Variables|Default|GetPossibleBossSpawnTags) _output_get :!class /Script/GameplayTags.BlueprintGameplayTagLibrary))

(event Custom|SpawnAllBossesWithTag
  (params (BossSpawnTag :type Gameplay%20Tag%20Structure))
  (for _array_element (Variables|Default|GetBossChallengeManagers)
    (Variables|/AIProgressBar/Blueprints/ChallengeManager.ChallengeManager_C|SetBossIDTag BossSpawnTag _array_element)
    (Class|ChallengeManager|Init _array_element)))

(event EventEndPlay
  (params (EndPlayReason :type EEndPlayReason%20Enum :default "Destroyed"))
  (Networking|SwitchHasAuthority
    (:Authority
      (bind _safezoneindicatorupdatedasynctask (Variables|SafeZone|GetSafeZoneIndicatorUpdatedAsyncTask)
        (:then
          (GameplayEventRouter|StopListeningtoStatefulEvent _safezoneindicatorupdatedasynctask)
          (bind _safezoneindicator (Variables|Default|GetSafeZoneIndicator)
            (:then
              (SafeZone|UnbindEventfromSafezoneStateChangedDelegate _safezoneindicator (EventDispatchers|CreateEvent :!event OnSafeZoneStateChanged)))
            (:else)))
        (:else)))))

(fn GetRandomManagerInNextSafeZone
  (if (and (< (Variables|Default|GetCountSpawned) (ScalableFloat|GetValueasInteger (Variables|Default|GetSpawnCount))) (Utilities|Array|IsNotEmpty (Variables|Default|GetBossChallengeManagers)))
    (for _array_element (Variables|Default|GetBossChallengeManagers)
      (if (SafeZone|IsinNextSafeZone (Variables|Default|GetSafeZoneIndicator) (Transformation|GetActorLocation _array_element))
        (Utilities|Array|Add (LocalVariables|GetChallengeManagersInsideSafeZone) _array_element)))
    (bind _element (ArrayMacros|GetRandomArrayElement (LocalVariables|GetChallengeManagersInsideSafeZone))
      (:%20%20
        (return _element))
      (:Empty%20Array
        (return 0)))
    (else
      (return 0))))

(fn GetValidBossTag
  (params (ChallengeManager :type Challenge%20Manager%20Simple%20Boss%20Progress%20Bar%20Object%20Reference))
  (bind _possiblebossspawntags (Variables|Default|GetPossibleBossSpawnTags))
  (bind _locationnametodatarowmap (Variables|Default|GetLocationNameToDataRowMap))
  (if (ScalableFloat|GetValueasBool (Variables|Default|GetCanCheckValidBossLocation))
    (LocalVariables|SetNewPossibleBossSpawnTags _possiblebossspawntags)
    ;[1 1 1] Loop Challenge Manager Tags to see if it has a location row attached
    (for _array_element (Variables|/Script/Engine.Actor|GetTags ChallengeManager)
      (if (.ReturnValue (Utilities|Map|Find _locationnametodatarowmap _array_element))
        ;[1 1 1] Find turned off tags for loaction and remove them from spawn pool
        (for (_array_element_2 _array_index) (Variables|Default|GetValidBossTags)
          (if (not (ScalableFloat|GetValueasBool (Utilities|Map|Find _locationnametodatarowmap _array_element) (Math|Conversions|ToFloat(Integer) _array_index)))
            (GameplayTags|RemoveGameplayTag (LocalVariables|GetNewPossibleBossSpawnTags) (.AssociatedBossTag_8_A43A80F049F2C88B343C20BBDFE7981E (Utilities|Struct|BreakValidBossStruct _array_element_2)) :!class /Script/GameplayTags.BlueprintGameplayTagLibrary)))
        (return (Utilities|Array|RandomArrayItem (GameplayTags|BreakGameplayTagContainer (LocalVariables|GetNewPossibleBossSpawnTags))))))
    (else
      (return (Utilities|Array|RandomArrayItem (GameplayTags|BreakGameplayTagContainer _possiblebossspawntags))))))

(fn InitilizeWithSpawnLocationTags
  (params (InSpawnedPOITags :type Gameplay%20Tag%20Container%20Structure))
  (Variables|Default|SetSpawnedPOILocations InSpawnedPOITags)
  (CallFunction|InitializeBossManager)
  (return))

(fn OnSafeZoneIndicatorSpawned
  (params (SafeZoneIndicator :type Fort%20Safe%20Zone%20Indicator%20Object%20Reference))
  ;[0 0 0] Bind delegate to safe zone state changes and trigger initial OnSafeZoneUpdated call
  (bind _output_get (Variables|Default|SetSafeZoneIndicator SafeZoneIndicator))
  (Utilities|IsValid _output_get
    (:Is%20Valid
      (SafeZone|BindEventtoSafezoneStateChangedDelegate _output_get (EventDispatchers|CreateEvent :!event OnSafeZoneStateChanged))
      (CallFunction|OnSafeZoneUpdated (SafeZone|GetCurrentPhaseIndex (Variables|Default|GetSafeZoneIndicator)) (SafeZone|GetSafeZoneState (Variables|Default|GetSafeZoneIndicator)))
      (bind _safezoneindicatorupdatedasynctask (Variables|SafeZone|GetSafeZoneIndicatorUpdatedAsyncTask)
        (:then
          (GameplayEventRouter|StopListeningtoStatefulEvent _safezoneindicatorupdatedasynctask))
        (:else)))))

(fn OnSafeZoneStateChanged
  (params (SafeZoneIndicator :type Fort%20Safe%20Zone%20Indicator%20Object%20Reference) (SafeZoneState :type EFortSafeZoneState%20Enum))
  ;[0 0 0 0.5] Trigger OnSafeZoneUpdated when the safe zone state changes. Additionally pass in current phase index
  (Utilities|IsValid SafeZoneIndicator
    (:Is%20Valid
      (CallFunction|OnSafeZoneUpdated (SafeZone|GetCurrentPhaseIndex SafeZoneIndicator) SafeZoneState))))

(fn OnSafeZoneUpdated
  (params (SafeZonePhase :type Integer) (SafeZoneState :type EFortSafeZoneState%20Enum))
  (bind _randombosschallengemanager (CallFunction|GetRandomManagerinNextSafeZone))
  (if (Math|Integer|InRange(Integer) SafeZonePhase (ScalableFloat|GetValueasInteger (Variables|SafeZone|GetMinSafeZonePhaseToSpawn)) (ScalableFloat|GetValueasInteger (Variables|SafeZone|GetMaxSafeZonePhaseToSpawn)))
    (if (< (Variables|Default|GetCountSpawned) (ScalableFloat|GetValueasInteger (Variables|Default|GetSpawnCount)))
      (if (== (Math|Conversions|ToInteger(Byte) (Variables|GetSafeZoneState)) (ScalableFloat|GetValueasInteger (Variables|SafeZone|GetSafeZoneStateToSpawn)))
        ;[0 0 0 0.5] Spawn random number of bosses for this phase. Limit random spawns based on MaxSpawnsPerSafeZonePhase.
        ; Clamp from 1 - SpawnCount
        (for _ (range (Utilities|ClampMaxInteger (Math|Random|RandomIntegerinRange 1 (ScalableFloat|GetValueasInteger (Variables|SafeZone|GetMaxSpawnsPerSafeZonePhase))) (ScalableFloat|GetValueasInteger (Variables|Default|GetSpawnCount))))
          ;[0 0 0 0.5] Initialize selected challenge manager with random boss data
          (Utilities|IsValid _randombosschallengemanager
            (:Is%20Valid
              (CallFunction|InitializeChallengeManager _randombosschallengemanager))
            (:Is%20Not%20Valid
              (Fortnite|FortShippingLog self "Failed to find valid boss challenge manager within safe zone"))))))
    (else
      (Fortnite|FortShippingLog self "Boss spawning cancelled. Outside of safe zone phase range"))))

(fn ConstructionScript)
